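# Niagara trail effect for the GoodMood pickaxe, drawing short-lived ribbon anim trails from skeletal mesh bones or sockets.
# BladeTrailMain spans one index pair with a time-of-day tinted color, while EarTrails runs two dark, shorter-lived trails on two other pairs.
# Both fade alpha over particle life and multiply it by a System.Age envelope, so the trails appear and vanish within about a quarter second of system start.
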NiagaraSystem /BRCosmetics/Weapons/FORT_Melee/Pickaxe_GoodMood/FX/NS_Pickaxe_GoodMood_Trails
dependencies {
 usedDataInterfaces [&NiagaraDataInterfaceColorCurve, &NiagaraDataInterfaceCurve]
 usedDynamicInputs []
 usedModules [
  &self:SystemSpawnScript.SystemScriptSource.SystemScriptGraph.NiagaraNodeAssignment_0.SetVariables_#1
  &SystemState
 ]
}
emitters {
 BladeTrailMain {
  emitterUpdateScript [
   &EmitterState()
   &FN_SpawnAnimTrail(Spawn Interval=0)
  ]
  particleSpawnScript [
   &FN_InitializeAnimTrailBoneIndecies(First Bone or Socket Index=1, Second Bone or Socket Index=0)
   &FN_InitializeAnimTrail(
    Color=&FNTimeOfDayColor(
     Night Color=<rgba 0.36 0.54 0.9 0.95>
     Morning Color=<rgba 0.9 0.81 0.831 0.86>
     Day Color=<rgba 1 1 1 0.9>
     Evening Color=<rgba 2 1.08 0.8 0.7>
    )
    Lifetime=0.11
    SkeletalMesh=$System.Skeletal Mesh
   )
   &AddVelocity(Velocity=<0 1000 0>) {enabled=false}
  ]
  particleUpdateScript [
   &FN_UpdateAnimTrail(Locked Particle RibbonWidth=1, SkeletalMesh=$User.SkeletalMesh) {module=FN_UpdateAnimTrail001}
   &ScaleColor(
    Scale Alpha=&FloatFromCurve(FloatCurve=DataInterface() {propertyValues {Curve {keys [
         {arriveTangent=-6.1981, interpMode=RCIM_Cubic, leaveTangent=-6.1981, tangentMode=RCTM_User, time=0, value=2}
         {arriveTangent=-1.8336, interpMode=RCIM_Cubic, leaveTangent=-1.8336, tangentMode=RCTM_User, time=0.1, value=1}
         {interpMode=RCIM_Cubic, tangentMode=RCTM_User, time=1, value=0}
        ]}}})
    Scale RGB=&MakeVectorFromLinearColorRGB(Color=&ColorFromCurve(ColorCurve=DataInterface() {propertyValues {
        AlphaCurve {keys []}
        BlueCurve {keys [<at 0.0912 1 RCIM_Cubic>,<at 0.9757 0.9>]}
        GreenCurve {keys [<at 0.0912 1 RCIM_Cubic>,<at 0.9757 0.9>]}
        RedCurve {keys [<at 0.0912 1 RCIM_Cubic>,<at 0.9757 0.9>]}
       }}))
    ScaleA=true
    ScaleRGB=true
   )
   &ScaleColor(
    Scale Alpha=&FloatFromCurve(
     FloatCurve=DataInterface() {propertyValues {Curve {keys [
         {interpMode=RCIM_Cubic, tangentMode=RCTM_User, time=0, value=0}
         {interpMode=RCIM_Cubic, tangentMode=RCTM_User, time=0.05, value=1}
         {interpMode=RCIM_Cubic, tangentMode=RCTM_User, time=0.16, value=1}
         {interpMode=RCIM_Cubic, tangentMode=RCTM_User, time=0.24, value=0}
        ]}}}
     CurveIndex=$System.Age
    )
    ScaleA=true
    ScaleRGB=true
   ) {module=ScaleColor001}
   &SolveForcesAndVelocity(Clamp Velocity=false, Limit Acceleration=false) {enabled=false}
  ]
  rendererClasses [&NiagaraRibbonRendererProperties]
  renderers [{rendererClass=&NiagaraRibbonRendererProperties}]
 }
 EarTrails {
  emitterUpdateScript [
   &EmitterState()
   &FN_SpawnAnimTrail(Number of Anim Trails=2, Spawn Interval=0)
  ]
  particleSpawnScript [
   &FN_InitializeAnimTrailBoneIndecies(First Bone or Socket Index=2, Second Bone or Socket Index=3)
   &FN_InitializeAnimTrailBoneIndecies(AnimTrail Selection=1, First Bone or Socket Index=4, Second Bone or Socket Index=5) {module=FN_InitializeAnimTrailBoneIndecies001}
   &FN_InitializeAnimTrail(AnimTrailWidth=1.2, Color=<rgba 0.0073 0.006 0.01 4>, Lifetime=0.06, SkeletalMesh=$System.Skeletal Mesh)
   &AddVelocity(Velocity=<0 1000 0>) {enabled=false}
  ]
  particleUpdateScript [
   &FN_UpdateAnimTrail(Locked Particle RibbonWidth=1, SkeletalMesh=$User.SkeletalMesh) {module=FN_UpdateAnimTrail001}
   &ScaleColor(
    Scale Alpha=&FloatFromCurve(FloatCurve=DataInterface() {propertyValues {Curve {keys [
         {arriveTangent=-6.1981, interpMode=RCIM_Cubic, leaveTangent=-6.1981, tangentMode=RCTM_User, time=0, value=2}
         {arriveTangent=-1.8336, interpMode=RCIM_Cubic, leaveTangent=-1.8336, tangentMode=RCTM_User, time=0.1, value=1}
         {interpMode=RCIM_Cubic, tangentMode=RCTM_User, time=1, value=0}
        ]}}})
    Scale RGB=&MakeVectorFromLinearColorRGB(Color=&ColorFromCurve(ColorCurve=DataInterface() {propertyValues {
        AlphaCurve {keys []}
        BlueCurve {keys [<at 0.0912 1 RCIM_Cubic>,<at 0.9757 0.9>]}
        GreenCurve {keys [<at 0.0912 1 RCIM_Cubic>,<at 0.9757 0.9>]}
        RedCurve {keys [<at 0.0912 1 RCIM_Cubic>,<at 0.9757 0.9>]}
       }}))
    ScaleA=true
    ScaleRGB=false
   )
   &ScaleColor(
    Scale Alpha=&FloatFromCurve(
     FloatCurve=DataInterface() {propertyValues {Curve {keys [
         {interpMode=RCIM_Cubic, tangentMode=RCTM_User, time=0, value=0}
         {interpMode=RCIM_Cubic, tangentMode=RCTM_User, time=0.05, value=1}
         {interpMode=RCIM_Cubic, tangentMode=RCTM_User, time=0.16, value=1}
         {interpMode=RCIM_Cubic, tangentMode=RCTM_User, time=0.24, value=0}
        ]}}}
     CurveIndex=$System.Age
    )
    ScaleA=true
    ScaleRGB=true
   ) {module=ScaleColor001}
   &SolveForcesAndVelocity(Clamp Velocity=false, Limit Acceleration=false) {enabled=false}
  ]
  rendererClasses [&NiagaraRibbonRendererProperties]
  renderers [{rendererClass=&NiagaraRibbonRendererProperties}]
 }
}
summary {
 systemName=NS_Pickaxe_GoodMood_Trails
 emitters [
  {emitterName=BladeTrailMain, rendererClasses [&NiagaraRibbonRendererProperties]}
  {emitterName=EarTrails, rendererClasses [&NiagaraRibbonRendererProperties]}
 ]
 userVariables [
  User.bisSwinging {type=NiagaraBool}
  User.NewVector3f {type=Vector3f, description="", defaultValue {struct=Vector3f, value=<0 0 0>}}
  User.RibbonIDOffset {type=NiagaraInt32}
  User.SkeletalMesh {
   type=NiagaraDataInterfaceSkeletalMesh
   defaultValue {
    struct=NiagaraExt_VariableValue_DataInterface
    value {dataInterface=&self:NiagaraDataInterfaceSkeletalMesh_1}
   }
  }
  User.Variant {type=NiagaraFloat, defaultValue=0, description=""}
 ]
}
