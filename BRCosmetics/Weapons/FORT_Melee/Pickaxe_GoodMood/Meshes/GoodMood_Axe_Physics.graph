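# Physics asset for the GoodMood pickaxe mesh, a cosmetic melee weapon in BRCosmetics.
# It defines no constraints between bodies, so no joints link any parts of the weapon.
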PhysicsAsset /BRCosmetics/Weapons/FORT_Melee/Pickaxe_GoodMood/Meshes/GoodMood_Axe_Physics
constraints []
