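# Animation Blueprint for the GoodMood pickaxe skeleton, adding secondary motion to its ear and hair bones.
# Cached poses chain AnimDynamics on ears and hair, interpolated additive ear offsets while jumping with enough X movement, then additive noise rotations on hair.
# The event graph updates hair gravity and four Perlin-style rotators shared across nine hair bones, damped in the front end.
# Weapon-type flags are set but unused by the AnimGraph.
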AnimBlueprint /BRCosmetics/Weapons/FORT_Melee/Pickaxe_GoodMood/Meshes/GoodMood_Axe_Skeleton_AnimBP
target_skeleton=&/GoodMood_Axe_Skeleton
graphs [
 AnimGraph {nodes [
   n1 = Get ambientRotatorA() -> (ambientRotatorA:Rotator)
   n2 = "Transform (Modify) Bone - Bone: dyn_hair_01_bck"(ComponentPose=Local To Component(LocalPose=Use cached pose 'CRCache_postFixes'()), Alpha=1, Translation="0,0,0", Rotation=@n1, Scale="1,1,1") {
    rotationMode=BMM_Additive
    rotationSpace=BCS_BoneSpace
    translationMode=BMM_Additive
    translationSpace=BCS_BoneSpace
    alphaBoolBlend {blendOption=Linear}
    boneToModify {boneName=dyn_hair_01_bck}
   }
   n3 = Get ambienRotatorB() -> (ambienRotatorB:Rotator)
   n4 = "Transform (Modify) Bone - Bone: dyn_hair_02_bck"(ComponentPose=@n2, Alpha=1, Translation="0,0,0", Rotation=@n3, Scale="1,1,1") {
    rotationMode=BMM_Additive
    rotationSpace=BCS_BoneSpace
    translationMode=BMM_Additive
    translationSpace=BCS_BoneSpace
    alphaBoolBlend {blendOption=Linear}
    boneToModify {boneName=dyn_hair_02_bck}
   }
   n5 = Get ambientRotatorC() -> (ambientRotatorC:Rotator)
   n6 = "Transform (Modify) Bone - Bone: dyn_hair_03_bck"(ComponentPose=@n4, Alpha=1, Translation="0,0,0", Rotation=@n5, Scale="1,1,1") {
    rotationMode=BMM_Additive
    rotationSpace=BCS_BoneSpace
    translationMode=BMM_Additive
    translationSpace=BCS_BoneSpace
    alphaBoolBlend {blendOption=Linear}
    boneToModify {boneName=dyn_hair_03_bck}
   }
   n7 = Get ambientRotatorD() -> (ambientRotatorD:Rotator)
   n8 = "Transform (Modify) Bone - Bone: dyn_hair_04_bck"(ComponentPose=@n6, Alpha=1, Translation="0,0,0", Rotation=@n7, Scale="1,1,1") {
    rotationMode=BMM_Additive
    rotationSpace=BCS_BoneSpace
    translationMode=BMM_Additive
    translationSpace=BCS_BoneSpace
    alphaBoolBlend {blendOption=Linear}
    boneToModify {boneName=dyn_hair_04_bck}
   }
   n9 = "Transform (Modify) Bone - Bone: dyn_hair_05_bck"(ComponentPose=@n8, Alpha=1, Translation="0,0,0", Rotation=@n1, Scale="1,1,1") {
    rotationMode=BMM_Additive
    rotationSpace=BCS_BoneSpace
    translationMode=BMM_Additive
    translationSpace=BCS_BoneSpace
    alphaBoolBlend {blendOption=Linear}
    boneToModify {boneName=dyn_hair_05_bck}
   }
   n10 = "Transform (Modify) Bone - Bone: dyn_hair_06_bck"(ComponentPose=@n9, Alpha=1, Translation="0,0,0", Rotation=@n3, Scale="1,1,1") {
    rotationMode=BMM_Additive
    rotationSpace=BCS_BoneSpace
    translationMode=BMM_Additive
    translationSpace=BCS_BoneSpace
    alphaBoolBlend {blendOption=Linear}
    boneToModify {boneName=dyn_hair_06_bck}
   }
   n11 = "Transform (Modify) Bone - Bone: dyn_hair_07_bck"(ComponentPose=@n10, Alpha=1, Translation="0,0,0", Rotation=@n5, Scale="1,1,1") {
    rotationMode=BMM_Additive
    rotationSpace=BCS_BoneSpace
    translationMode=BMM_Additive
    translationSpace=BCS_BoneSpace
    alphaBoolBlend {blendOption=Linear}
    boneToModify {boneName=dyn_hair_07_bck}
   }
   n12 = "Transform (Modify) Bone - Bone: dyn_hair_08_bck"(ComponentPose=@n11, Alpha=1, Translation="0,0,0", Rotation=@n7, Scale="1,1,1") {
    rotationMode=BMM_Additive
    rotationSpace=BCS_BoneSpace
    translationMode=BMM_Additive
    translationSpace=BCS_BoneSpace
    alphaBoolBlend {blendOption=Linear}
    boneToModify {boneName=dyn_hair_08_bck}
   }
   n13 = "Transform (Modify) Bone - Bone: dyn_hair_09_bck"(ComponentPose=@n12, Alpha=1, Translation="0,0,0", Rotation=@n7, Scale="1,1,1") {
    rotationMode=BMM_Additive
    rotationSpace=BCS_BoneSpace
    translationMode=BMM_Additive
    translationSpace=BCS_BoneSpace
    alphaBoolBlend {blendOption=Linear}
    boneToModify {boneName=dyn_hair_09_bck}
   }
   Output Pose(Result=Component To Local(ComponentPose=@n13)) {layerGroup=None}
   n14 = Anim Dynamics - Bone: dyn_ear_01_l!(
    ComponentPose=Local To Component(LocalPose=Use cached pose 'CRCache_Input'())
    Alpha=1
    GravityOverride:Vector="0,0,-500"
    ExternalForce="0,0,0"
   ) -> (Pose) {
    angularDampingOverride=1
    angularSpringConstant=20000
    bAngularSpring=true
    bGravityOverrideInSimSpace=true
    bOverrideAngularDamping=true
    bOverrideLinearDamping=true
    bUseGravityOverride=true
    gravityScale=0.6
    linearDampingOverride=1
    alphaBoolBlend {blendOption=Linear}
    boundBone {boneName=dyn_ear_01_l}
    componentLinearAccScale=<1 1 1.2>
    componentLinearVelScale=<1 1 1.2>
    gravityOverride {z=-500}
    physicsBodyDefinitions [{
      boundBone {boneName=dyn_ear_01_l}
      boxExtents=<20 30 20>
      constraintSetup {angularLimitsMax=<5 30 15>, angularLimitsMin=<-5 -30 -10>, angularTarget=<1 0 0>}
      localJointOffset=<2 0 0>
     }]
    relativeSpaceBone {boneName=b_AR_Root}
   }
   n15 = Anim Dynamics - Bone: dyn_ear_02_l!(ComponentPose=@n14, Alpha=1, GravityOverride:Vector="0,0,-500", ExternalForce="0,0,0") -> (Pose) {
    angularDampingOverride=1
    angularSpringConstant=20000
    bAngularSpring=true
    bGravityOverrideInSimSpace=true
    bOverrideAngularDamping=true
    bOverrideLinearDamping=true
    bUseGravityOverride=true
    gravityScale=0.6
    linearDampingOverride=1
    alphaBoolBlend {blendOption=Linear}
    boundBone {boneName=dyn_ear_02_l}
    componentLinearAccScale=<1 1 1.5>
    componentLinearVelScale=<1 1 1.5>
    gravityOverride {z=-500}
    physicsBodyDefinitions [{
      boundBone {boneName=dyn_ear_02_l}
      boxExtents=<30 40 30>
      constraintSetup {angularLimitsMax=<10 50 25>, angularLimitsMin=<-10 -50 -10>, angularTarget=<1 0 0>}
      localJointOffset=<4 0 0>
     }]
    relativeSpaceBone {boneName=b_AR_Root}
   }
   n16 = Anim Dynamics - Bone: dyn_ear_03_l!(ComponentPose=@n15, Alpha=1, GravityOverride:Vector="0,0,-500", ExternalForce="0,0,0") -> (Pose) {
    angularDampingOverride=1
    angularSpringConstant=20000
    bAngularSpring=true
    bGravityOverrideInSimSpace=true
    bOverrideAngularDamping=true
    bOverrideLinearDamping=true
    bUseGravityOverride=true
    gravityScale=0.6
    linearDampingOverride=1
    alphaBoolBlend {blendOption=Linear}
    boundBone {boneName=dyn_ear_03_l}
    componentLinearAccScale=<1 1 1.7>
    componentLinearVelScale=<1 1 1.7>
    gravityOverride {z=-500}
    physicsBodyDefinitions [{
      boundBone {boneName=dyn_ear_03_l}
      boxExtents=<40 50 40>
      constraintSetup {angularLimitsMax=<10 40 45>, angularLimitsMin=<-10 -45 -20>, angularTarget=<1 0 0>}
      localJointOffset=<5 0 0>
     }]
    relativeSpaceBone {boneName=b_AR_Root}
   }
   n17 = Anim Dynamics - Bone: dyn_ear_01_r!(ComponentPose=@n16, Alpha=1, GravityOverride:Vector="0,0,-980", ExternalForce="0,0,0") -> (Pose) {
    angularDampingOverride=1
    angularSpringConstant=20000
    bAngularSpring=true
    bGravityOverrideInSimSpace=true
    bOverrideAngularDamping=true
    bOverrideLinearDamping=true
    bUseGravityOverride=true
    gravityScale=0.6
    linearDampingOverride=1
    alphaBoolBlend {blendOption=Linear}
    boundBone {boneName=dyn_ear_01_r}
    componentLinearAccScale=<1 1 1.2>
    componentLinearVelScale=<1 1 2>
    gravityOverride {z=-980}
    physicsBodyDefinitions [{
      boundBone {boneName=dyn_ear_01_r}
      boxExtents=<25 35 25>
      constraintSetup {angularLimitsMax=<5 20 1>, angularLimitsMin=<-5 -20 -1>, angularTarget=<1 0 0>}
      localJointOffset=<2 0 0>
     }]
    relativeSpaceBone {boneName=b_AR_Root}
   }
   n18 = Anim Dynamics - Bone: dyn_ear_02_r!(ComponentPose=@n17, Alpha=1, GravityOverride:Vector="0,0,-980", ExternalForce="0,0,0") -> (Pose) {
    angularDampingOverride=1
    angularSpringConstant=20000
    bAngularSpring=true
    bGravityOverrideInSimSpace=true
    bOverrideAngularDamping=true
    bOverrideLinearDamping=true
    bUseGravityOverride=true
    gravityScale=0.6
    linearDampingOverride=1
    alphaBoolBlend {blendOption=Linear}
    boundBone {boneName=dyn_ear_02_r}
    componentLinearAccScale=<1 1 1.5>
    componentLinearVelScale=<1 1 1.5>
    gravityOverride {z=-980}
    physicsBodyDefinitions [{
      boundBone {boneName=dyn_ear_02_r}
      boxExtents=<35 45 35>
      constraintSetup {angularLimitsMax=<15 35 10>, angularLimitsMin=<-15 -45 -10>, angularTarget=<1 0 0>}
      localJointOffset=<2 0 0>
     }]
    relativeSpaceBone {boneName=b_AR_Root}
   }
   n19 = Anim Dynamics - Bone: dyn_ear_03_r!(ComponentPose=@n18, Alpha=1, GravityOverride:Vector="0,0,-980", ExternalForce="0,0,0") -> (Pose) {
    angularDampingOverride=1
    angularSpringConstant=20000
    bAngularSpring=true
    bGravityOverrideInSimSpace=true
    bOverrideAngularDamping=true
    bOverrideLinearDamping=true
    bUseGravityOverride=true
    gravityScale=0.6
    linearDampingOverride=1
    alphaBoolBlend {blendOption=Linear}
    boundBone {boneName=dyn_ear_03_r}
    componentLinearAccScale=<1 1 1.7>
    componentLinearVelScale=<1 1 1.7>
    gravityOverride {z=-980}
    physicsBodyDefinitions [{
      boundBone {boneName=dyn_ear_03_r}
      boxExtents=<45 55 45>
      constraintSetup {angularLimitsMax=<15 45 10>, angularLimitsMin=<-15 -50 -15>, angularTarget=<1 0 0>}
      localJointOffset=<5 0 0>
     }]
    relativeSpaceBone {boneName=b_AR_Root}
   }
   Save cached pose 'CRCache_earDyns'(Pose=Component To Local(ComponentPose=@n19)) {cacheName=CRCache_earDyns}
   Save cached pose 'CRCache_Input'(Pose=Input Pose()) {cacheName=CRCache_Input}
   n20 = Get GravityOverride()
   n21 = Anim Dynamics - Bone: dyn_hair_01_bck!(ComponentPose=Local To Component(LocalPose=Use cached pose 'CRCache_earDyns'()), Alpha=1, GravityOverride:Vector=@n20, ExternalForce="0,0,0") -> (Pose) {
    angularDampingOverride=1
    angularSpringConstant=20000
    bAngularSpring=true
    bOverrideAngularDamping=true
    bUseGravityOverride=true
    gravityScale=0.35
    simulationSpace=BoneRelative
    alphaBoolBlend {blendOption=Linear}
    boundBone {boneName=dyn_hair_01_bck}
    componentLinearAccScale=<1 1 1>
    componentLinearVelScale=<1 1 1>
    gravityOverride {z=-980}
    physicsBodyDefinitions [{
      boundBone {boneName=dyn_hair_01_bck}
      boxExtents=<20 30 30>
      constraintSetup {angularLimitsMax=<15 15 15>, angularLimitsMin=<-15 -25 -5>, angularTarget=<1 0 0>}
      localJointOffset=<4 0 0>
     }]
    relativeSpaceBone {boneName=b_AR_Root}
   }
   n22 = Anim Dynamics - Bone: dyn_hair_02_bck!(ComponentPose=@n21, Alpha=1, GravityOverride:Vector=@n20, ExternalForce="0,0,0") -> (Pose) {
    angularDampingOverride=1
    angularSpringConstant=20000
    bAngularSpring=true
    bOverrideAngularDamping=true
    bUseGravityOverride=true
    gravityScale=0.35
    simulationSpace=BoneRelative
    alphaBoolBlend {blendOption=Linear}
    boundBone {boneName=dyn_hair_02_bck}
    componentAppliedLinearAccClamp=<3000 3000 3000>
    componentLinearAccScale=<0.6 0.6 0.6>
    componentLinearVelScale=<0.6 0.6 0.6>
    gravityOverride {z=-980}
    physicsBodyDefinitions [{
      boundBone {boneName=dyn_hair_02_bck}
      boxExtents=<20 30 30>
      constraintSetup {angularLimitsMax=<15 15 15>, angularLimitsMin=<-15 -25 -5>, angularTarget=<1 0 0>}
      localJointOffset=<4 0 0>
     }]
    relativeSpaceBone {boneName=b_AR_Root}
   }
   n23 = Anim Dynamics - Bone: dyn_hair_03_bck!(ComponentPose=@n22, Alpha=1, GravityOverride:Vector=@n20, ExternalForce="0,0,0") -> (Pose) {
    angularDampingOverride=1
    angularSpringConstant=20000
    bAngularSpring=true
    bOverrideAngularDamping=true
    bUseGravityOverride=true
    gravityScale=0.35
    simulationSpace=BoneRelative
    alphaBoolBlend {blendOption=Linear}
    boundBone {boneName=dyn_hair_03_bck}
    componentLinearAccScale=<1 1 1>
    componentLinearVelScale=<1 1 1>
    gravityOverride {z=-980}
    physicsBodyDefinitions [{
      boundBone {boneName=dyn_hair_03_bck}
      boxExtents=<20 30 30>
      constraintSetup {angularLimitsMax=<15 15 15>, angularLimitsMin=<-15 -25 -5>, angularTarget=<1 0 0>}
      localJointOffset=<4 0 0>
     }]
    relativeSpaceBone {boneName=b_AR_Root}
   }
   n24 = Anim Dynamics - Bone: dyn_hair_04_bck!(ComponentPose=@n23, Alpha=1, GravityOverride:Vector=@n20, ExternalForce="0,0,0") -> (Pose) {
    angularDampingOverride=1
    angularSpringConstant=20000
    bAngularSpring=true
    bOverrideAngularDamping=true
    bUseGravityOverride=true
    gravityScale=0.35
    simulationSpace=BoneRelative
    alphaBoolBlend {blendOption=Linear}
    boundBone {boneName=dyn_hair_04_bck}
    componentLinearAccScale=<1 1 1>
    componentLinearVelScale=<1 1 1>
    gravityOverride {z=-980}
    physicsBodyDefinitions [{
      boundBone {boneName=dyn_hair_04_bck}
      boxExtents=<20 30 30>
      constraintSetup {angularLimitsMax=<15 15 15>, angularLimitsMin=<-15 -25 -5>, angularTarget=<1 0 0>}
      localJointOffset=<4 0 0>
     }]
    relativeSpaceBone {boneName=b_AR_Root}
   }
   n25 = Anim Dynamics - Bone: dyn_hair_05_bck!(ComponentPose=@n24, Alpha=1, GravityOverride:Vector=@n20, ExternalForce="0,0,0") -> (Pose) {
    angularDampingOverride=1
    angularSpringConstant=20000
    bAngularSpring=true
    bOverrideAngularDamping=true
    bUseGravityOverride=true
    gravityScale=0.35
    simulationSpace=BoneRelative
    alphaBoolBlend {blendOption=Linear}
    boundBone {boneName=dyn_hair_05_bck}
    componentLinearAccScale=<1 1 1>
    componentLinearVelScale=<1 1 1>
    gravityOverride {z=-980}
    physicsBodyDefinitions [{
      boundBone {boneName=dyn_hair_05_bck}
      boxExtents=<20 30 30>
      constraintSetup {angularLimitsMax=<15 15 15>, angularLimitsMin=<-15 -25 -5>, angularTarget=<1 0 0>}
      localJointOffset=<4 0 0>
     }]
    relativeSpaceBone {boneName=b_AR_Root}
   }
   n26 = Anim Dynamics - Bone: dyn_hair_06_bck!(ComponentPose=@n25, Alpha=1, GravityOverride:Vector=@n20, ExternalForce="0,0,0") -> (Pose) {
    angularDampingOverride=1
    angularSpringConstant=20000
    bAngularSpring=true
    bOverrideAngularDamping=true
    bUseGravityOverride=true
    gravityScale=0.35
    simulationSpace=BoneRelative
    alphaBoolBlend {blendOption=Linear}
    boundBone {boneName=dyn_hair_06_bck}
    componentLinearAccScale=<1 1 1>
    componentLinearVelScale=<1 1 1>
    gravityOverride {z=-980}
    physicsBodyDefinitions [{
      boundBone {boneName=dyn_hair_06_bck}
      boxExtents=<20 30 30>
      constraintSetup {angularLimitsMax=<15 15 15>, angularLimitsMin=<-15 -25 -5>, angularTarget=<1 0 0>}
      localJointOffset=<4 0 0>
     }]
    relativeSpaceBone {boneName=b_AR_Root}
   }
   n27 = Anim Dynamics - Bone: dyn_hair_07_bck!(ComponentPose=@n26, Alpha=1, GravityOverride:Vector=@n20, ExternalForce="0,0,0") -> (Pose) {
    angularDampingOverride=1
    angularSpringConstant=20000
    bAngularSpring=true
    bOverrideAngularDamping=true
    bUseGravityOverride=true
    gravityScale=0.35
    simulationSpace=BoneRelative
    alphaBoolBlend {blendOption=Linear}
    boundBone {boneName=dyn_hair_07_bck}
    componentLinearAccScale=<1 1 1>
    componentLinearVelScale=<1 1 1>
    gravityOverride {z=-980}
    physicsBodyDefinitions [{
      boundBone {boneName=dyn_hair_07_bck}
      boxExtents=<20 30 30>
      constraintSetup {angularLimitsMax=<15 15 15>, angularLimitsMin=<-15 -25 -5>, angularTarget=<1 0 0>}
      localJointOffset=<4 0 0>
     }]
    relativeSpaceBone {boneName=b_AR_Root}
   }
   n28 = Anim Dynamics - Bone: dyn_hair_08_bck!(ComponentPose=@n27, Alpha=1, GravityOverride:Vector=@n20, ExternalForce="0,0,0") -> (Pose) {
    angularDampingOverride=1
    angularSpringConstant=20000
    bAngularSpring=true
    bOverrideAngularDamping=true
    bUseGravityOverride=true
    gravityScale=0.35
    simulationSpace=BoneRelative
    alphaBoolBlend {blendOption=Linear}
    boundBone {boneName=dyn_hair_08_bck}
    componentLinearAccScale=<1 1 1>
    componentLinearVelScale=<1 1 1>
    gravityOverride {z=-980}
    physicsBodyDefinitions [{
      boundBone {boneName=dyn_hair_08_bck}
      boxExtents=<20 30 30>
      constraintSetup {angularLimitsMax=<15 15 15>, angularLimitsMin=<-15 -25 -5>, angularTarget=<1 0 0>}
      localJointOffset=<4 0 0>
     }]
    relativeSpaceBone {boneName=b_AR_Root}
   }
   n29 = Anim Dynamics - Bone: dyn_hair_09_bck!(ComponentPose=@n28, Alpha=1, GravityOverride:Vector=@n20, ExternalForce="0,0,0") -> (Pose) {
    angularDampingOverride=1
    angularSpringConstant=20000
    bAngularSpring=true
    bOverrideAngularDamping=true
    bUseGravityOverride=true
    gravityScale=0.35
    simulationSpace=BoneRelative
    alphaBoolBlend {blendOption=Linear}
    boundBone {boneName=dyn_hair_09_bck}
    componentLinearAccScale=<1 1 1>
    componentLinearVelScale=<1 1 1>
    gravityOverride {z=-980}
    physicsBodyDefinitions [{
      boundBone {boneName=dyn_hair_09_bck}
      boxExtents=<20 30 30>
      constraintSetup {angularLimitsMax=<15 15 15>, angularLimitsMin=<-15 -25 -5>, angularTarget=<1 0 0>}
      localJointOffset=<4 0 0>
     }]
    relativeSpaceBone {boneName=b_AR_Root}
   }
   Save cached pose 'dynCache'(Pose=Component To Local(ComponentPose=@n29)) {cacheName=dynCache}
   n30 = Get isJump() -> (isJump:double)
   n31 = "Transform (Modify) Bone - Bone: base_ear_01_l"(
    ComponentPose=Local To Component(LocalPose=Use cached pose 'dynCache'())
    Alpha=@n30
    Translation="0,2,0"
    Rotation="-10,10,0"
    Scale="1,1,1"
   ) {
    rotationMode=BMM_Additive
    rotationSpace=BCS_BoneSpace
    translationMode=BMM_Additive
    translationSpace=BCS_BoneSpace
    alphaBoolBlend {blendOption=Linear}
    alphaScaleBiasClamp {bInterpResult=true, interpSpeedDecreasing=3, interpSpeedIncreasing=3}
    boneToModify {boneName=base_ear_01_l}
    rotation {pitch=-10, yaw=10}
    translation {y=2}
   }
   n32 = "Transform (Modify) Bone - Bone: dyn_ear_02_l"(ComponentPose=@n31, Alpha=@n30, Translation="0,3,-2", Rotation="-5,5,0", Scale="1,1,1") {
    rotationMode=BMM_Additive
    rotationSpace=BCS_BoneSpace
    translationMode=BMM_Additive
    translationSpace=BCS_BoneSpace
    alphaBoolBlend {blendOption=Linear}
    alphaScaleBiasClamp {bInterpResult=true, interpSpeedDecreasing=3, interpSpeedIncreasing=3}
    boneToModify {boneName=dyn_ear_02_l}
    rotation {pitch=-5, yaw=5}
    translation {y=3, z=-2}
   }
   n33 = "Transform (Modify) Bone - Bone: dyn_ear_03_l"(ComponentPose=@n32, Alpha=@n30, Translation="0,2,-2", Rotation="-20,-15,0", Scale="1,1,1") {
    rotationMode=BMM_Additive
    rotationSpace=BCS_BoneSpace
    translationMode=BMM_Additive
    translationSpace=BCS_BoneSpace
    alphaBoolBlend {blendOption=Linear}
    alphaScaleBiasClamp {bInterpResult=true, interpSpeedDecreasing=3, interpSpeedIncreasing=3}
    boneToModify {boneName=dyn_ear_03_l}
    rotation {pitch=-20, yaw=-15}
    translation {y=2, z=-2}
   }
   n34 = "Transform (Modify) Bone - Bone: base_ear_01_r"(ComponentPose=@n33, Alpha=@n30, Translation="0,0,0", Rotation="20,10,0", Scale="1,1,1") {
    rotationMode=BMM_Additive
    rotationSpace=BCS_BoneSpace
    translationMode=BMM_Additive
    translationSpace=BCS_BoneSpace
    alphaBoolBlend {blendOption=Linear}
    alphaScaleBiasClamp {bInterpResult=true, interpSpeedDecreasing=3, interpSpeedIncreasing=3}
    boneToModify {boneName=base_ear_01_r}
    rotation {pitch=20, yaw=10}
   }
   n35 = "Transform (Modify) Bone - Bone: dyn_ear_02_r"(ComponentPose=@n34, Alpha=@n30, Translation="0,0,5", Rotation="0,15,0", Scale="1,1,1") {
    rotationMode=BMM_Additive
    rotationSpace=BCS_BoneSpace
    translationMode=BMM_Additive
    translationSpace=BCS_BoneSpace
    alphaBoolBlend {blendOption=Linear}
    alphaScaleBiasClamp {bInterpResult=true, interpSpeedDecreasing=3, interpSpeedIncreasing=3}
    boneToModify {boneName=dyn_ear_02_r}
    rotation {yaw=15}
    translation {z=5}
   }
   n36 = "Transform (Modify) Bone - Bone: dyn_ear_03_r"(ComponentPose=@n35, Alpha=@n30, Translation="0,0,5", Rotation="0,15,0", Scale="1,1,1") {
    rotationMode=BMM_Additive
    rotationSpace=BCS_BoneSpace
    translationMode=BMM_Additive
    translationSpace=BCS_BoneSpace
    alphaBoolBlend {blendOption=Linear}
    alphaScaleBiasClamp {bInterpResult=true, interpSpeedDecreasing=3, interpSpeedIncreasing=3}
    boneToModify {boneName=dyn_ear_03_r}
    rotation {yaw=15}
    translation {z=5}
   }
   Save cached pose 'CRCache_postFixes'(Pose=Component To Local(ComponentPose=@n36)) {cacheName=CRCache_postFixes}
  ]}
 EventGraph {nodes [
   Event BlueprintUpdateAnimation()
   > n1 = Sequence!(execute) -> (then_0, then_1, then_2:Exec, then_3:Exec, then_4:Exec, then_5:Exec, then_6:Exec, then_7:Exec, then_8:Exec, then_9:Exec)
   n2 = GetGameTimeInSeconds()
   n3 = Get IsFrontLvL() -> (IsFrontLvL:double)
   n4 = RandomRotatorFunction!(
    execute=@n1.then_1
    GameTimeInSeconds:double=@n2
    PerlinScale:double=2
    Frequency:Rotator="Scale Rotator (Float)"(A:Rotator="6,4,5.5", B:float=@n3) -> (ReturnValue:Rotator=~) {is=title K2Node_PromotableOperator}
    Amplitude:Rotator="Scale Rotator (Float)"(A="5,4,5", B=@n3) -> (ReturnValue=~)
    Offset:Rotator="0, 0, 0"
    TimeOffset:double=0
   ) -> (then, RandomRotator:Rotator)
   > Set ambientRotatorA(ambientRotatorA:Rotator=@n4) -> (Output_Get:Rotator)
   n5 = RandomRotatorFunction(
    execute=@n1.then_2
    GameTimeInSeconds=@n2
    PerlinScale=2.4
    Frequency="Scale Rotator (Float)"(A="5,6,4", B=@n3) -> (ReturnValue=~)
    Amplitude="Scale Rotator (Float)"(A="4,3.5,3", B=@n3) -> (ReturnValue=~)
    Offset="0, 0, 0"
    TimeOffset=0
   )
   > Set ambienRotatorB(ambienRotatorB:Rotator=@n5) -> (Output_Get:Rotator)
   n6 = RandomRotatorFunction(
    execute=@n1.then_3
    GameTimeInSeconds=@n2
    PerlinScale=1.8
    Frequency="Scale Rotator (Float)"(A="4,5,3", B=@n3) -> (ReturnValue=~)
    Amplitude="Scale Rotator (Float)"(A="4,4.4,4", B=@n3) -> (ReturnValue=~)
    Offset="0, 0, 0"
    TimeOffset=0
   )
   > Set ambientRotatorC(ambientRotatorC:Rotator=@n6) -> (Output_Get:Rotator)
   n7 = RandomRotatorFunction(
    execute=@n1.then_4
    GameTimeInSeconds=@n2
    PerlinScale=2.2
    Frequency="Scale Rotator (Float)"(A="3.5,4,4.5", B=@n3) -> (ReturnValue=~)
    Amplitude="Scale Rotator (Float)"(A="5.5,3.5,5", B=@n3) -> (ReturnValue=~)
    Offset="0, 0, 0"
    TimeOffset=0
   )
   > Set ambientRotatorD(ambientRotatorD:Rotator=@n7) -> (Output_Get:Rotator)
   n8 = Branch(execute=@n1.then_6, Condition=Get bIsInFrontEnd())
   > Set IsFrontLvL(IsFrontLvL:double=0.5) -> (Output_Get:double)
   Set IsFrontLvL(execute=@n8.else, IsFrontLvL=1)
   n9 = Gravity Override State Function DEPRECATED(
    execute=@n1.then_0
    CharacterPartAnimInstance=Self-Reference()
    GameTimeInSeconds=GetGameTimeInSeconds()
    GravityOverrideParams=Get Gravity Override Params()
   )
   > Set GravityOverride(GravityOverride=@n9)
   n10 = PlayerMovement(execute=@n1.then_5, AnimInstance=Self-Reference(), XDotProductCoefficient=100, YDotProductCoefficient=100, ZDotProductCoefficient=100)
   n11 = BreakVector(InVec=MakeVector(X=@n10, Y=@n10.YMovement))
   > n12 = Branch(Condition=AND Boolean(A=Get bIsJumping(), B="float > float"(A=@n11, B=50)))
   > Set isJump(isJump:double=1) -> (Output_Get:double)
   Set isJump(execute=@n12.else, isJump=0)
   n13 = Cast To FortPlayerAnimInstance(execute=@n1.then_8, Object=GetAnimInstance(self=GetOwnerSkeletalMeshComponent()))
   > n14 = Switch on EFortWeaponCoreAnimation!(execute, Selection:EFortWeaponCoreAnimation Enum=Get WeaponCoreAnim(self=@n13)) -> (
    Melee:Exec
    Pistol:Exec
    Shotgun:Exec
    PaperBlueprint:Exec
    Rifle:Exec
    MeleeOneHand:Exec
    MachinePistol:Exec
    RocketLauncher:Exec
    GrenadeLauncher:Exec
    GoingCommando:Exec
    AssaultRifle:Exec
    TacticalShotgun:Exec
    SniperRifle:Exec
    TrapPlacement:Exec
    ShoulderLauncher:Exec
    AbilityDecoTool:Exec
    Crossbow:Exec
    C4:Exec
    RemoteControl:Exec
    DualWield:Exec
    AR_BullPup:Exec
    AR_ForwardGrip:Exec
    MedPackPaddles:Exec
    SMG_P90:Exec
    AR_DrumGun:Exec
    Consumable_Small:Exec
    Consumable_Large:Exec
    Balloon:Exec
    MountedTurret:Exec
    CreativeTool:Exec
    ExplosiveBow:Exec
    AshtonIndigo:Exec
    AshtonChicago:Exec
    MeleeDualWield:Exec
    Unarmed:Exec
   )
   Branch(execute=[
     @n14.AR_BullPup
     @n14.AR_DrumGun
     @n14.AR_ForwardGrip
     @n14.AbilityDecoTool
     @n14.AshtonChicago
     @n14.AshtonIndigo
     @n14.AssaultRifle
     @n14.Balloon
     @n14.C4
     @n14.Consumable_Large
     @n14.Consumable_Small
     @n14.CreativeTool
     @n14.Crossbow
     @n14.DualWield
     @n14.ExplosiveBow
     @n14.GoingCommando
     @n14.GrenadeLauncher
     @n14.MachinePistol
     @n14.MedPackPaddles
     @n14.MeleeDualWield
     @n14.MeleeOneHand
     @n14.MountedTurret
     @n14.PaperBlueprint
     @n14.Pistol
     @n14.RemoteControl
     @n14.Rifle
     @n14.RocketLauncher
     @n14.SMG_P90
     @n14.Shotgun
     @n14.ShoulderLauncher
     @n14.SniperRifle
     @n14.TacticalShotgun
     @n14.TrapPlacement
     @n14.Unarmed
    ])
   > Set MeeleWeapon(MeeleWeapon:bool=false) -> (Output_Get:bool)
   > Set ActivateBool(ActivateBool:bool=true) -> (Output_Get:bool)
   Branch(execute=@n14.Melee)
   > Set MeeleWeapon(MeeleWeapon=true)
   > Delay(Duration=0.2)
   > Set ActivateBool(ActivateBool=false)
   TryGetPawnOwner()
  ]}
]
variables [
 ActivateBool {type=boolean}
 ambienRotatorB {type=Rotator}
 ambientRotatorA {type=Rotator}
 ambientRotatorC {type=Rotator}
 ambientRotatorD {type=Rotator}
 Gravity Override Params {type=GravityOverrideParamsStruct}
 GravityOverride {type=Vector}
 IsFrontLvL {type=number}
 isJump {type=number}
 meeleVar {type=number}
 MeeleWeapon {type=boolean}
]
